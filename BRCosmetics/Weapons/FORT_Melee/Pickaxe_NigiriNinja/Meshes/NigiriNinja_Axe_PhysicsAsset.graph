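# Physics asset for the NigiriNinja two-handed pickaxe mesh, a melee cosmetic weapon, giving it a simplified physics shape.
# A single body on the Melee_TwoHanded_Handle bone uses one long capsule, offset along the bone's Z axis and standing nearly parallel to it.
# It has no constraints, so the asset defines no jointed or separately simulated parts.
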PhysicsAsset /BRCosmetics/Weapons/FORT_Melee/Pickaxe_NigiriNinja/Meshes/NigiriNinja_Axe_PhysicsAsset
bodies {Melee_TwoHanded_Handle [{shapeType=Capsule, length=91.3003, radius=9.404, center=<1.001 -1.5991 43.3425>, rotation=<rot 0.459 -90 0.0023>}]}
constraints []
